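# Translucent, unlit, two-sided material for Niagara mesh particles, drawing a fire explosion shockwave on a radial ground-wave mesh.
# Emissive color multiplies two plasma textures through a smoothstep, a red tint and particle color, then feeds an emissive adjustment function.
# The alpha mask pushes noise, distorted by a second panning noise, along clamped V via the Offset dynamic parameter.
# View-angle fading, a sine-based V edge falloff and particle alpha shape that mask.
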Material /Game/Effects/Niagara/MasterMaterials/ShockWave/M_Fire_Explosion_GroundWave
graph {
 properties {
  bAutomaticallySetUsageInEditor=false
  bUsedWithNiagaraMeshParticles=true
  blendMode=BLEND_Translucent
  pixelDepthOffsetMode=PDOM_Legacy
  previewMesh=/Game/Effects/Fort_Effects/Mesh/Shapes/SM_Radial_Ground_Wave
  shadingModel=MSM_Unlit
  twoSided=true
  usedShadingModels=Unlit
 }
 nodes [
  n1 = Constant3Vector() {constant=<rgba 1 0.0573 0.0573 1>}
  n2 = TextureCoordinate() {uTiling=3, vTiling=0.5}
  n3 = Panner(Coordinate=@n2)
  n4 = Panner(Coordinate=@n2)
  n5 = SmoothStep(Value=Multiply(
    A=TextureSample(UVs=@n3) {samplerType=SAMPLERTYPE_LinearColor, texture=/Game/Effects/Fort_Effects/Emotes/UltraLight/Texture/T_Plasma_D_C}.R
    B=TextureSample(UVs=@n4) {samplerType=SAMPLERTYPE_LinearGrayscale, texture=/Game/Effects/Fort_Effects/Emotes/UltraLight/Texture/T_PlasmaCoils}.R
   )) {constMax=0.5}
  n6 = ParticleColor()
  MaterialFunctionCall(Emissive=Multiply(A=Multiply(A=@n1, B=@n5), B=@n6.RGB)) {materialFunction=&MF_FX_EmissiveAdjustments}
  n7 = Panner(
   Coordinate=Add(A=TextureCoordinate() {uTiling=3, vTiling=0.4}, B=NamedRerouteUsage())
   Time=DynamicParameter() {defaultValue=<rgba 0.8 1 1 1>, paramNames [Offset,Param2,Param3,Param4]}.Offset
  ) {speedY=1}
  n8 = AppendVector(A=ComponentMask(None=@n7) {r=true}, B=Clamp(None=ComponentMask(None=@n7) {g=true}))
  n9 = Multiply(
   A=Power(Base=TextureSample(UVs=@n8) {samplerType=SAMPLERTYPE_Grayscale, texture=&LowResBlurredNoise}.R) {constExponent=1}
   B=MaterialFunctionCall("FadeSensitivity (optional)"=Constant() {r=1}) {materialFunction=&FadeBasedOnViewAngle}.Result
  )
  n10 = TextureCoordinate()
  n11 = Power(Base=Sine(None=ComponentMask(None=@n10) {g=true}) {period=2}) {constExponent=3}
  Multiply(A=Multiply(A=@n6.A, B=Multiply(A=@n9, B=@n11))) {constB=1.5}
  n12 = Panner(Coordinate=TextureCoordinate() {uTiling=2, vTiling=0.2}) {speedY=0.3}
  NamedRerouteDeclaration(None=Multiply(A=TextureSample(UVs=@n12) {samplerType=SAMPLERTYPE_Grayscale, texture=&LowResBlurredNoise}.R) {constB=0.2}) {name=Disturb, nodeColor=<rgba 1 0.0752 0 1>}
 ]
}
