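# Unlit, two-sided translucent master material for Niagara mesh particles, drawing an impact shockwave tinted by particle color.
# The dynamic parameter Offset scrolls a flame texture along V, clipped with Min against smoke streaks sampled only along U.
# That mask is smoothstepped, then multiplied by view-angle fade, a sine falloff at both V ends and particle alpha.
# Particle RGB passes through MF_FX_EmissiveAdjustments for the glow.
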Material /Game/Effects/Niagara/MasterMaterials/ShockWave/M_Impact_Wave
graph {
 properties {
  bAutomaticallySetUsageInEditor=false
  bUsedWithNiagaraMeshParticles=true
  blendMode=BLEND_Translucent
  pixelDepthOffsetMode=PDOM_Legacy
  shadingModel=MSM_Unlit
  twoSided=true
  usedShadingModels=Unlit
 }
 nodes [
  n1 = ParticleColor()
  MaterialFunctionCall(Emissive=@n1.RGB) {materialFunction=&MF_FX_EmissiveAdjustments}
  n2 = Panner(
   Coordinate=Add(A=TextureCoordinate() {uTiling=4, vTiling=0.2}) {constB=0}
   Time=DynamicParameter() {defaultValue=<rgba 0.8 1 1 1>, paramNames [Offset,Param2,Param3,Param4]}.Offset
  ) {speedY=1}
  n3 = TextureSample(UVs=AppendVector(A=ComponentMask(None=@n2) {r=true}, B=ComponentMask(None=@n2) {g=true})) {
   samplerSource=SSM_Wrap_WorldGroupSettings
   samplerType=SAMPLERTYPE_LinearGrayscale
   texture=/Game/Effects/Fort_Effects/Emotes/UltraLight/Texture/T_FluidFlameTile_Greyscale
  }
  n4 = TextureCoordinate() {uTiling=4, vTiling=0}
  n5 = Min(
   A=Power(Base=@n3.R) {constExponent=0.5}
   B=TextureSample(UVs=Panner(Coordinate=@n4)) {texture=&T_Inky_Smoke_Tile}.R
  )
  n6 = Saturate(None=Multiply(B=SmoothStep(Value=@n5) {constMin=-0.1}) {constA=1})
  n7 = MaterialFunctionCall("FadeSensitivity (optional)"=Constant() {r=2}) {materialFunction=&FadeBasedOnViewAngle}
  n8 = Sine(None=ComponentMask(None=TextureCoordinate()) {g=true}) {period=2}
  Multiply(A=@n1.A, B=Multiply(A=Multiply(A=@n6, B=@n7.Result), B=Power(Base=@n8)))
 ]
}
