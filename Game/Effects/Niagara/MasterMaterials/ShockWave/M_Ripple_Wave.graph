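# Unlit translucent master material for Niagara sprites and mesh particles, drawing a rippling shockwave ring with adjustable emissive and particle color.
# Radial UVs with optional per-particle random offsets pan textures, distorted by scrolling noise.
# Three offset samples form a color-fringed emissive blended by parameter.
# Noise-warped outer and inner sphere masks carve the ring alpha, followed by depth fade (non-mobile deferred path), view-angle fade and optional edge mask.
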Material /Game/Effects/Niagara/MasterMaterials/ShockWave/M_Ripple_Wave
graph {
 properties {
  bUsedWithNiagaraMeshParticles=true
  bUsedWithNiagaraSprites=true
  blendMode=BLEND_Translucent
  pixelDepthOffsetMode=PDOM_Legacy
  shadingModel=MSM_Unlit
  usedShadingModels=Unlit
 }
 nodes [
  n1 = VectorParameter() {parameterName=MainTex_Scale_UV, defaultValue=<rgba 1 1 0 0>}
  n2 = Add(
   A=Multiply(A=ParticleRandom(), B=ScalarParameter() {defaultValue=1, parameterName=Particle_Random UV Offset On/Off})
   B=Multiply(
    A=MaterialFunctionCall() {materialFunction=&VectorToRadialValue}.Radial Coordinates
    B=ComponentMask(None=@n1.RGB) {g=true, r=true}
   )
  )
  n3 = ComponentMask(None=VectorParameter() {parameterName=MainTex_Speed, defaultValue=<rgba 0 -0.4 0 0>}.RGB) {g=true, r=true}
  n4 = VectorParameter() {parameterName=Noise_Scale_UV, defaultValue=<rgba 1 1 0 0>}
  n5 = Add(
   A=Multiply(A=ParticleRandom(), B=ScalarParameter() {defaultValue=1, parameterName=Particle_Random UV Offset On/Off})
   B=Multiply(
    A=MaterialFunctionCall() {materialFunction=&VectorToRadialValue}.Radial Coordinates
    B=ComponentMask(None=@n4.RGB) {g=true, r=true}
   )
  )
  n6 = VectorParameter() {parameterName=noise speed, defaultValue=<rgba 1 1 0 1>}
  n7 = Constant()
  n8 = TextureSample!(
   UVs=Multiply(A=Panner(Coordinate=@n5) {speedX=0.2, speedY=-1}, B=ComponentMask(None=@n6.RGB) {g=true, r=true})
   Tex
   "DDX(UVs)"=@n7
   "DDY(UVs)"=@n7
   Apply View MipBias
  ) {
   mipValueMode=TMVM_Derivative
   samplerType=SAMPLERTYPE_LinearGrayscale
   texture=/Game/Effects/Fort_Effects/Emotes/UltraLight/Texture/T_DirectionalNoise_Greyscale
  }
  n9 = Multiply(A=@n8.R, B=ScalarParameter() {defaultValue=0.2, parameterName=Noise_Intensity}) {constB=0.2}
  n10 = Add(A=Multiply(A=Panner(Coordinate=@n2, Speed=@n3) {speedY=-0.4}, B=Constant2Vector() {g=1, r=1}), B=@n9)
  n11 = ScalarParameter() {defaultValue=0.02, parameterName=Color offset}
  n12 = Constant()
  n13 = TextureSampleParameter2D!(UVs=Add(A=@n11, B=@n10), "DDX(UVs)"=@n12, "DDY(UVs)"=@n12, Apply View MipBias) {
   mipValueMode=TMVM_Derivative
   parameterName=Tex Noise
   samplerType=SAMPLERTYPE_LinearGrayscale
   texture=/Game/Effects/Fort_Effects/Emotes/UltraLight/Texture/T_PlasmaCoils
  }
  n14 = Constant()
  n15 = TextureSampleParameter2D(UVs=@n10, "DDX(UVs)"=@n14, "DDY(UVs)"=@n14) {
   mipValueMode=TMVM_Derivative
   parameterName=Tex Noise
   samplerType=SAMPLERTYPE_LinearGrayscale
   texture=&T_PlasmaCoils
  }
  n16 = Constant()
  n17 = TextureSampleParameter2D(UVs=Subtract(A=@n10, B=@n11), "DDX(UVs)"=@n16, "DDY(UVs)"=@n16) {
   mipValueMode=TMVM_Derivative
   parameterName=Tex Noise
   samplerType=SAMPLERTYPE_LinearGrayscale
   texture=&T_PlasmaCoils
  }
  n18 = MaterialFunctionCall(X=@n13.R, Y=@n15.R, Z=@n17.R) {materialFunction=&MakeFloat3}
  n19 = Constant()
  n20 = TextureSampleParameter2D(UVs=@n10, "DDX(UVs)"=@n19, "DDY(UVs)"=@n19) {
   mipValueMode=TMVM_Derivative
   parameterName=Tex Base
   samplerType=SAMPLERTYPE_LinearGrayscale
   texture=/Game/Characters/Player/Male/Medium/Bodies/M_MED_Citadel/FX/T_Citadel_Noise_Dissolve_Handpaint
  }
  n21 = Multiply(A=@n20.R, B=ScalarParameter() {defaultValue=0.7, parameterName=Center Alpha}) {constB=0.7}
  n22 = Add(A=Desaturation(None=Power(Base=@n18.Result) {constExponent=3}, Fraction=ScalarParameter() {defaultValue=-0.2, parameterName=Des}), B=@n21)
  n23 = ParticleColor()
  MaterialFunctionCall(Emissive=LinearInterpolate(A=@n23.RGB, B=Multiply(A=@n22, B=@n23.RGB), Alpha=ScalarParameter() {parameterName=Use Colorful Texture})) {materialFunction=&MF_FX_EmissiveAdjustments}
  n24 = Add(A=Power(Base=@n15.R, Exp=ScalarParameter() {defaultValue=3, parameterName=Alpha power}) {constExponent=3}, B=@n21)
  n25 = OneMinus(None=SphereMask(A=TextureCoordinate(), B=Constant() {r=0.5}, Radius=ScalarParameter() {defaultValue=0.5, parameterName=Central_Glow}) {attenuationRadius=0.5, hardnessPercent=0})
  n26 = Power(
   Base=Power(Base=Add(A=@n24) {constB=0}, Exp=LinearInterpolate(Alpha=@n25) {constA=-0.1, constB=2})
   Exp=ScalarParameter() {defaultValue=1.2, parameterName=Alpha_Power}
  ) {constExponent=1.2}
  n27 = Add(A=TextureCoordinate(), B=Constant2Vector() {g=-0.07, r=-0.1})
  n28 = TextureCoordinate()
  n29 = Constant2Vector() {r=-0.05}
  n30 = Add(
   A=Multiply(A=ParticleRandom(), B=ScalarParameter() {defaultValue=1, parameterName=Particle_Random UV Offset On/Off})
   B=MaterialFunctionCall() {materialFunction=&VectorToRadialValue}.Radial Coordinates
  )
  n31 = Constant()
  n32 = TextureSample(UVs=Multiply(A=Panner(Coordinate=@n30) {speedX=0.2, speedY=1}, B=Constant2Vector() {g=0.5, r=1}), "DDX(UVs)"=@n31, "DDY(UVs)"=@n31) {
   mipValueMode=TMVM_Derivative
   samplerType=SAMPLERTYPE_LinearGrayscale
   texture=&T_SimplePerlin_Greyscale
  }
  n33 = Multiply(A=@n32.R) {constB=0.3}
  n34 = SphereMask(
   A=Add(A=Add(A=@n28, B=@n29), B=Multiply(A=@n33) {constB=0.2})
   B=Constant() {r=0.5}
   Radius=Add(A=ScalarParameter() {defaultValue=0.3, parameterName=Inner_Mask_Radius}, B=NamedRerouteUsage())
  ) {attenuationRadius=0.3, hardnessPercent=30}
  n35 = Multiply(
   A=SphereMask(A=Add(A=@n27, B=@n9), B=Constant() {r=0.5}, Hardness=ScalarParameter() {parameterName=SphereMask_Hardness}) {attenuationRadius=0.45, hardnessPercent=0}
   B=Power(Base=OneMinus(None=@n34))
  )
  n36 = Multiply(A=@n23.A, B=Add(A=Multiply(A=@n26, B=@n35), B=ScalarParameter() {parameterName=Alpha_Overall}))
  n37 = DepthFade(Opacity=@n36, FadeDistance=ScalarParameter() {defaultValue=10, parameterName=FadeDistance}) {fadeDistanceDefault=10}
  n38 = ScalarParameter() {parameterName="FadeSensitivity (optional) (S)"}
  n39 = Multiply(A=SphereMask(A=TextureCoordinate(), B=Constant() {r=0.5}) {attenuationRadius=0.5, hardnessPercent=0}) {constB=5}
  Multiply(
   A=Multiply(
    A=ShadingPathSwitch(Default=@n37, Deferred=@n37, Forward=@n36, Mobile=@n36)
    B=MaterialFunctionCall("FadeSensitivity (optional)"=@n38) {materialFunction=&FadeBasedOnViewAngle}.Result
   )
   B=LinearInterpolate(B=Saturate(None=@n39), Alpha=ScalarParameter() {parameterName=Use sphere edge mask}) {constA=1}
  )
  n40 = DynamicParameter() {defaultValue=<rgba 0 0 0 0>, paramNames [Inner_Radius, U Offset, V Offset, Param4]}
  NamedRerouteDeclaration(None=@n40.Inner_Radius) {name=Inner_Radius, nodeColor=<rgba 0.1719 0 1 1>}
  NamedRerouteDeclaration(None=@n40.U Offset) {name=U_Offset, nodeColor=<rgba 0.5703 0 1 1>}
  NamedRerouteDeclaration(None=@n40.V Offset) {name=V_Offset, nodeColor=<rgba 0 0.3906 1 1>}
  Saturate(None=Add())
 ]
}
