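# Unlit translucent master material for shock wave effects on Niagara mesh particles, rendered two-sided.
# Particle color feeds emissive through an adjustment function, while its alpha scales the wave opacity mask.
# The mask blends a spike texture toward two mixed smoke trail textures via UseSoftWave, scrolled by a dynamic offset parameter and shaped by a powered V gradient.
# Depth fade and view-angle fade soften the resulting opacity.
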Material /Game/Effects/Niagara/MasterMaterials/ShockWave/M_Spine_ShockWave
graph {
 properties {
  bUsedWithNiagaraMeshParticles=true
  blendMode=BLEND_Translucent
  pixelDepthOffsetMode=PDOM_Legacy
  previewMesh=/Game/Weapons/FORT_Melee/Effects/FlintlockMelee/S_WaveMesh
  shadingModel=MSM_Unlit
  twoSided=true
  usedShadingModels=Unlit
 }
 nodes [
  n1 = ParticleColor()
  MaterialFunctionCall(Emissive=@n1.RGB) {materialFunction=&MF_FX_EmissiveAdjustments}
  n2 = TextureCoordinate()
  n3 = Constant()
  n4 = DynamicParameter() {defaultValue=<rgba 0 1 1 1>, paramNames [offset,Param2,Param3,Param4]}
  n5 = Panner(Coordinate=Add(A=TextureCoordinate() {uTiling=2}, B=AppendVector(A=@n3, B=@n4.offset)))
  n6 = TextureCoordinate() {uTiling=2}
  n7 = AppendVector(A=Constant(), B=DynamicParameter() {defaultValue=<rgba 0 1 1 1>, paramNames [offset,Param2,Param3,Param4]}.offset)
  n8 = TextureSample(UVs=Panner(Coordinate=Add(A=@n6, B=@n7))) {texture=/Game/Effects/Fort_Effects/Textures/Smoke/T_Smoke_Trail_Soft_2}
  n9 = TextureCoordinate() {uTiling=3}
  n10 = AppendVector(A=Constant(), B=DynamicParameter() {defaultValue=<rgba 0 1 1 1>, paramNames [offset,Param2,Param3,Param4]}.offset)
  n11 = TextureSample(UVs=Panner(Coordinate=Add(A=@n9, B=@n10))) {texture=/Game/Effects/Fort_Effects/Textures/Smoke/T_Lobber_Trail_Mask}
  n12 = Multiply(
   A=LinearInterpolate(
    A=TextureSample(UVs=@n5) {texture=/Game/Effects/Niagara/MasterMaterials/ShockWave/Texture/T_Wave_Spikes}.R
    B=LinearInterpolate(A=@n8.R, B=@n11.R)
    Alpha=ScalarParameter() {parameterName=UseSoftWave}
   )
   B=Power(Base=ComponentMask(None=@n2) {g=true})
  )
  Multiply(
   A=DepthFade(Opacity=Multiply(A=@n1.A, B=@n12), FadeDistance=ScalarParameter() {parameterName=FadeDistance})
   B=MaterialFunctionCall("FadeSensitivity (optional)"=ScalarParameter() {parameterName="FadeSensitivity (optional) (S)"}) {materialFunction=&FadeBasedOnViewAngle}.Result
  )
 ]
}
