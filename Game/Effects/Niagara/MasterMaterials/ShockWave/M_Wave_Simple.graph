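# Unlit, translucent, two-sided Niagara master material for shockwave effects on sprite and mesh particles.
# A tiled, panning noise channel drives opacity and blends ColorA to ColorB, with particle color tinting the result before emissive adjustments.
# An optional panning mask texture modulates alpha, with tiling and speed packed into vector parameters for both textures.
# Opacity is shaped by a powered sine band along V, then view-angle and depth fades.
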Material /Game/Effects/Niagara/MasterMaterials/ShockWave/M_Wave_Simple
graph {
 properties {
  bAutomaticallySetUsageInEditor=false
  bUsedWithNiagaraMeshParticles=true
  bUsedWithNiagaraSprites=true
  blendMode=BLEND_Translucent
  shadingModel=MSM_Unlit
  twoSided=true
  usedShadingModels=Unlit
 }
 nodes [
  n1 = ParticleColor()
  n2 = TextureCoordinate()
  n3 = VectorParameter() {parameterName=BaseTexture, defaultValue=<rgba 2 0.5 0 0.3>}
  n4 = AppendVector(A=@n3.R, B=@n3.R)
  n5 = TextureSampleParameter2D(UVs=Panner(Coordinate=Multiply(A=@n2, B=@n4), Speed=AppendVector(A=@n3.B, B=@n3.B))) {parameterName=BaseTexture, samplerType=SAMPLERTYPE_LinearColor, texture=&T_NoisePack_LinearColor}
  n6 = ChannelMaskParameter(None=@n5.RGBA) {group=P2_AlphaAndColorTint, maskChannel=Green, parameterName=AlphaChannel, defaultValue=<rgba 0 1 0 0>}
  n7 = Multiply(A=@n1.A, B=@n6.R)
  n8 = TextureCoordinate()
  n9 = VectorParameter() {parameterName=MaskTexture, defaultValue=<rgba 2 0.5 0 0.3>}
  n10 = AppendVector(A=@n9.R, B=@n9.R)
  n11 = TextureSampleParameter2D(UVs=Panner(Coordinate=Multiply(A=@n8, B=@n10), Speed=AppendVector(A=@n9.B, B=@n9.B))) {parameterName=MaskTexture, samplerType=SAMPLERTYPE_Grayscale, texture=/Game/Weapons/FORT_Melee/Pickaxe_Cube_Ninja_Male/FX/T_CellNoise}
  n12 = LinearInterpolate(
   A=@n7
   B=Multiply(A=@n7, B=ChannelMaskParameter(None=@n11.RGBA) {group=P2_AlphaAndColorTint, parameterName=Mask_AlphaChannel}.R)
   Alpha=ScalarParameter() {parameterName=Use_MaskTexture}
  )
  n13 = ComponentMask(None=TextureCoordinate()) {g=true}
  n14 = Power(Base=Multiply(A=Sine(None=@n13) {period=2})) {constExponent=5}
  n15 = ScalarParameter() {parameterName=MeshFade}
  DepthFade(
   Opacity=Multiply(
    A=Multiply(A=@n12, B=@n14)
    B=MaterialFunctionCall("FadeSensitivity (optional)"=@n15) {materialFunction=&FadeBasedOnViewAngle}.Result
   )
   FadeDistance=ScalarParameter() {defaultValue=2, parameterName=DepthFade}
  )
  n16 = VectorParameter() {parameterName=ColorA, defaultValue=<rgba 0.5104 0.0114 0 1>}
  n17 = VectorParameter() {parameterName=ColorB, defaultValue=<rgba 5 0.4803 0.45 1>}
  MaterialFunctionCall(Emissive=Multiply(A=LinearInterpolate(A=@n16.RGB, B=@n17.RGB, Alpha=@n6.R), B=@n1.RGB)) {materialFunction=&MF_FX_EmissiveAdjustments}
 ]
}
